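# Physics asset for the F_MED_Bandolier face accessory mesh, defining jiggle bodies for the female character's hair and earring.
# Capsule bodies on dyn_ bones hang from a small head capsule through swing- and twist-limited constraints.
# The large bangs form a three-link chain with the tightest limits, and the left hair bulk branches into two strands.
# The earring, both mid bangs and the upper strand each attach directly to the head.
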PhysicsAsset /BRCosmetics/Characters/Player/Female/Medium/Bodies/F_MED_Bandolier/Meshes/Parts/F_MED_Bandolier_FaceAcc_Physics
bodies {
 dyn_earring [{shapeType=Capsule, length=2.2277, radius=0.6388, center=<1.8488 0 0>, rotation=<rot 90 0 0>}]
 dyn_hair_bulk_l [{shapeType=Capsule, length=5, radius=3, center=<2.5 0 0>, rotation=<rot 90 0 0>}]
 dyn_hair_bulk_strand_A_l [{shapeType=Capsule, length=2, radius=2, center=<1 0 0>, rotation=<rot 90 0 0>}]
 dyn_hair_bulk_strand_B_l [{shapeType=Capsule, length=4, radius=2, center=<1 0 0>, rotation=<rot 90 0 0>}]
 dyn_large_bang_01 [{shapeType=Capsule, length=3, radius=2, center=<2 0 0>, rotation=<rot 90 0 0>}]
 dyn_large_bang_02 [{shapeType=Capsule, length=3.5, radius=1.5, center=<1.9185 0.4624 0>, rotation=<rot 90 0 -10.0006>}]
 dyn_large_bang_03 [{shapeType=Capsule, length=2.25, radius=1, center=<0.8772 0 0.0001>, rotation=<rot 90 0 0>}]
 dyn_mid_bang_01 [{shapeType=Capsule, length=2, radius=1, center=<1 0 0>, rotation=<rot 90 0 0>}]
 dyn_mid_bang_02 [{shapeType=Capsule, length=2, radius=1, center=<1 0 0>, rotation=<rot 90 0 0>}]
 dyn_upper_strand [{shapeType=Capsule, length=2, radius=2, center=<2 0 0>, rotation=<rot 90 0 0>}]
 head [{shapeType=Capsule, length=1, radius=1}]
}
constraints [
 {bone1Name=dyn_hair_bulk_l, bone2Name=head, swing1LimitDegrees=20, swing2LimitDegrees=20, twistLimitDegrees=5}
 {bone1Name=dyn_hair_bulk_strand_A_l, bone2Name=dyn_hair_bulk_l, swing1LimitDegrees=20, swing2LimitDegrees=20, twistLimitDegrees=5}
 {bone1Name=dyn_hair_bulk_strand_B_l, bone2Name=dyn_hair_bulk_l, swing1LimitDegrees=20, swing2LimitDegrees=20, twistLimitDegrees=5}
 {bone1Name=dyn_large_bang_01, bone2Name=head, swing1LimitDegrees=10, swing2LimitDegrees=10, twistLimitDegrees=1}
 {bone1Name=dyn_large_bang_02, bone2Name=dyn_large_bang_01, swing1LimitDegrees=10, swing2LimitDegrees=10, twistLimitDegrees=1}
 {bone1Name=dyn_large_bang_03, bone2Name=dyn_large_bang_02, swing1LimitDegrees=10, swing2LimitDegrees=10, twistLimitDegrees=1}
 {bone1Name=dyn_mid_bang_01, bone2Name=head, swing1LimitDegrees=20, swing2LimitDegrees=20, twistLimitDegrees=5}
 {bone1Name=dyn_mid_bang_02, bone2Name=head, swing1LimitDegrees=20, swing2LimitDegrees=20, twistLimitDegrees=5}
 {bone1Name=dyn_earring, bone2Name=head, swing1LimitDegrees=20, swing2LimitDegrees=20, twistLimitDegrees=5}
 {bone1Name=dyn_upper_strand, bone2Name=head, swing1LimitDegrees=20, swing2LimitDegrees=20, twistLimitDegrees=5}
]
