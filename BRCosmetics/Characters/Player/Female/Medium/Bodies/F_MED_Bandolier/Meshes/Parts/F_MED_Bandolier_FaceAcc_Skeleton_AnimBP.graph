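# Post-style AnimBP for the female medium Bandolier face accessory mesh, adding secondary motion to an earring, hair bulk, strands and bangs.
# Copies the pose from the attached parent mesh, then runs a serial chain of single-bone AnimDynamics nodes to the output.
# Each node uses root-relative simulation, reduced gravity, heavy angular damping and tight per-bone angular limits.
# A RigidBody node with the FaceAcc physics asset sits unconnected, and the event graph holds only unwired default nodes.
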AnimBlueprint /BRCosmetics/Characters/Player/Female/Medium/Bodies/F_MED_Bandolier/Meshes/Parts/F_MED_Bandolier_FaceAcc_Skeleton_AnimBP
target_skeleton=&/F_MED_Bandolier_FaceAcc_Skeleton
graphs [
 AnimGraph {nodes [
   RigidBody() {
    overridePhysicsAsset=&/F_MED_Bandolier_FaceAcc_Physics
    simulationSpace=BaseBoneSpace
    alphaBoolBlend {blendOption=Linear}
    baseBoneRef {boneName=spine_05}
   }
   n1 = Anim Dynamics - Bone: dyn_earring(ComponentPose=Local To Component(LocalPose=Copy Pose From Mesh() {bUseAttachedParent=true}), Alpha=1, ExternalForce="0,0,0") {
    angularDampingOverride=0.9999
    bGravityOverrideInSimSpace=true
    bOverrideAngularDamping=true
    gravityScale=0.35
    simulationSpace=RootRelative
    alphaBoolBlend {blendOption=Linear}
    boundBone {boneName=dyn_earring}
    physicsBodyDefinitions [{
      sphereCollisionRadius=0
      boundBone {boneName=dyn_earring}
      boxExtents=<2 10 10>
      constraintSetup {angularLimitsMax=<0 45 0>, angularLimitsMin=<0 -30 -30>, angularTarget=<0 0 0>}
      localJointOffset=<2 0 0>
     }]
   }
   n2 = Anim Dynamics - Bone: dyn_hair_bulk_l(ComponentPose=@n1, Alpha=1, ExternalForce="0,0,0") {
    angularDampingOverride=0.9999
    bGravityOverrideInSimSpace=true
    bOverrideAngularDamping=true
    gravityScale=0.35
    simulationSpace=RootRelative
    alphaBoolBlend {blendOption=Linear}
    boundBone {boneName=dyn_hair_bulk_l}
    physicsBodyDefinitions [{
      sphereCollisionRadius=0
      boundBone {boneName=dyn_hair_bulk_l}
      boxExtents=<4 20 20>
      constraintSetup {angularLimitsMax=<0 5 5>, angularLimitsMin=<0 -30 -5>, angularTarget=<0 0 0>}
      localJointOffset=<4 0 0>
     }]
   }
   n3 = Anim Dynamics - Bone: dyn_hair_bulk_strand_A_l(ComponentPose=@n2, Alpha=1, ExternalForce="0,0,0") {
    angularDampingOverride=0.9999
    bGravityOverrideInSimSpace=true
    bOverrideAngularDamping=true
    gravityScale=0.35
    simulationSpace=RootRelative
    alphaBoolBlend {blendOption=Linear}
    boundBone {boneName=dyn_hair_bulk_strand_A_l}
    physicsBodyDefinitions [{
      sphereCollisionRadius=0
      boundBone {boneName=dyn_hair_bulk_strand_A_l}
      boxExtents=<2 10 10>
      constraintSetup {angularLimitsMax=<0 5 5>, angularLimitsMin=<0 -30 -5>, angularTarget=<0 0 0>}
      localJointOffset=<2 0 0>
     }]
   }
   n4 = Anim Dynamics - Bone: dyn_hair_bulk_strand_B_l(ComponentPose=@n3, Alpha=1, ExternalForce="0,0,0") {
    angularDampingOverride=0.9999
    bGravityOverrideInSimSpace=true
    bOverrideAngularDamping=true
    gravityScale=0.35
    simulationSpace=RootRelative
    alphaBoolBlend {blendOption=Linear}
    boundBone {boneName=dyn_hair_bulk_strand_B_l}
    physicsBodyDefinitions [{
      sphereCollisionRadius=0
      boundBone {boneName=dyn_hair_bulk_strand_B_l}
      boxExtents=<2 10 10>
      constraintSetup {angularLimitsMax=<0 5 5>, angularLimitsMin=<0 -30 -5>, angularTarget=<0 0 0>}
      localJointOffset=<2 0 0>
     }]
   }
   n5 = Anim Dynamics - Bone: dyn_large_bang_01(ComponentPose=@n4, Alpha=1, ExternalForce="0,0,0") {
    angularDampingOverride=0.9999
    bGravityOverrideInSimSpace=true
    bOverrideAngularDamping=true
    gravityScale=0.35
    simulationSpace=RootRelative
    alphaBoolBlend {blendOption=Linear}
    boundBone {boneName=dyn_large_bang_01}
    physicsBodyDefinitions [{
      sphereCollisionRadius=0
      boundBone {boneName=dyn_large_bang_01}
      boxExtents=<2 10 10>
      constraintSetup {angularLimitsMax=<0 10 0>, angularLimitsMin=<0 -2 0>, angularTarget=<0 0 0>}
      localJointOffset=<0.5 0 0>
     }]
   }
   n6 = Anim Dynamics - Bone: dyn_large_bang_02(ComponentPose=@n5, Alpha=1, ExternalForce="0,0,0") {
    angularDampingOverride=0.9999
    bGravityOverrideInSimSpace=true
    bOverrideAngularDamping=true
    gravityScale=0.35
    simulationSpace=RootRelative
    alphaBoolBlend {blendOption=Linear}
    boundBone {boneName=dyn_large_bang_02}
    physicsBodyDefinitions [{
      sphereCollisionRadius=0
      boundBone {boneName=dyn_large_bang_02}
      boxExtents=<2 10 10>
      constraintSetup {angularLimitsMax=<0 10 0>, angularLimitsMin=<0 -2 0>, angularTarget=<0 0 0>}
      localJointOffset=<0.5 0 0>
     }]
   }
   n7 = Anim Dynamics - Bone: dyn_large_bang_03(ComponentPose=@n6, Alpha=1, ExternalForce="0,0,0") {
    angularDampingOverride=0.9999
    bGravityOverrideInSimSpace=true
    bOverrideAngularDamping=true
    gravityScale=0.35
    simulationSpace=RootRelative
    alphaBoolBlend {blendOption=Linear}
    boundBone {boneName=dyn_large_bang_03}
    physicsBodyDefinitions [{
      sphereCollisionRadius=0
      boundBone {boneName=dyn_large_bang_03}
      boxExtents=<2 10 10>
      constraintSetup {angularLimitsMax=<0 10 0>, angularLimitsMin=<0 -2 0>, angularTarget=<0 0 0>}
      localJointOffset=<0.5 0 0>
     }]
   }
   n8 = Anim Dynamics - Bone: dyn_mid_bang_01(ComponentPose=@n7, Alpha=1, ExternalForce="0,0,0") {
    angularDampingOverride=0.9999
    bGravityOverrideInSimSpace=true
    bOverrideAngularDamping=true
    gravityScale=0.35
    simulationSpace=RootRelative
    alphaBoolBlend {blendOption=Linear}
    boundBone {boneName=dyn_mid_bang_01}
    physicsBodyDefinitions [{
      sphereCollisionRadius=0
      boundBone {boneName=dyn_mid_bang_01}
      boxExtents=<2 10 10>
      constraintSetup {angularLimitsMax=<0 10 0>, angularLimitsMin=<0 -2 0>, angularTarget=<0 0 0>}
      localJointOffset=<0.5 0 0>
     }]
   }
   n9 = Anim Dynamics - Bone: dyn_mid_bang_02(ComponentPose=@n8, Alpha=1, ExternalForce="0,0,0") {
    angularDampingOverride=0.9999
    bGravityOverrideInSimSpace=true
    bOverrideAngularDamping=true
    gravityScale=0.35
    simulationSpace=RootRelative
    alphaBoolBlend {blendOption=Linear}
    boundBone {boneName=dyn_mid_bang_02}
    physicsBodyDefinitions [{
      sphereCollisionRadius=0
      boundBone {boneName=dyn_mid_bang_02}
      boxExtents=<2 10 10>
      constraintSetup {angularLimitsMax=<0 10 0>, angularLimitsMin=<0 -2 0>, angularTarget=<0 0 0>}
      localJointOffset=<0.5 0 0>
     }]
   }
   n10 = Anim Dynamics - Bone: dyn_upper_strand(ComponentPose=@n9, Alpha=1, ExternalForce="0,0,0") {
    angularDampingOverride=0.9999
    bGravityOverrideInSimSpace=true
    bOverrideAngularDamping=true
    gravityScale=0.35
    simulationSpace=RootRelative
    alphaBoolBlend {blendOption=Linear}
    boundBone {boneName=dyn_upper_strand}
    physicsBodyDefinitions [{
      sphereCollisionRadius=0
      boundBone {boneName=dyn_upper_strand}
      boxExtents=<2 10 10>
      constraintSetup {angularLimitsMax=<0 10 0>, angularLimitsMin=<0 -2 0>, angularTarget=<0 0 0>}
      localJointOffset=<0.5 0 0>
     }]
   }
   Output Pose(Result=Component To Local(ComponentPose=@n10)) {layerGroup=None}
  ]}
 EventGraph {nodes [
   TryGetPawnOwner()
   Event BlueprintUpdateAnimation()
  ]}
]
variables []
